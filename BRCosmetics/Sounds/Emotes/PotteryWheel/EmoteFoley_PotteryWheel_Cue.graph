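# Sound cue for the Pottery Wheel emote foley in Battle Royale cosmetics.
# A mixer plays three layers together: a one-shot start sound, a wheel loop delayed one second to follow it, and a looping foley layer from the beginning.
# It uses emote attenuation, foley sound class and concurrency, keeps playing when silent, and sends to a music audio-analysis submix.
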SoundCue /BRCosmetics/Sounds/Emotes/PotteryWheel/EmoteFoley_PotteryWheel_Cue
root=Mixer_0
nodes [
 {id=Delay_0, children [WavePlayer_1], properties {delayMax=1, delayMin=1}}
 {id=Mixer_0, children [WavePlayer_0,Delay_0,WavePlayer_2], properties {inputVolume [1,1,1]}}
 {id=WavePlayer_0, properties {soundWaveAssetPtr=&/IP_Emote_PotteryWheel_Start}}
 {id=WavePlayer_1, properties {bLooping=true, soundWaveAssetPtr=&/IP_Emote_PotteryWheel_LP}}
 {id=WavePlayer_2, properties {bLooping=true, soundWaveAssetPtr=&/IP_Emote_PotteryWheel_Foley}}
]
properties {
 attenuationSettings=&Fort_Emotes_Athena_Att
 audioPropertiesSheet=/BRCosmetics/Sounds/Settings/PropertySheets/Emotes/APS_Cosmetic_Emote_Foley
 maxDistance=1600
 soundClassObject=/Game/Sounds/FortSoundClassesAndModes/Fort_Foley_Emote
 virtualizationMode=PlayWhenSilent
 volumeMultiplier=1
 concurrencyOverrides {maxCount=3}
 concurrencySet [/Game/Sounds/Concurrency_Rules/Athena_EmoteFoley_Conc]
 soundSubmixSends [{minSendLevel=0, soundSubmix=&Fortnite_AudioAnalysis_Music}]
}
